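# Material function that adds recolorable patterns to the Disguise Black medium male player body.
# The SkinFX_Mask channels blend primary, secondary and accent color parameters, with optional camo-style darkening of the secondary color.
# The result tints incoming base color and replaces base and subsurface color where red or green mask is present.
# Subsurface is scaled by a cloth intensity parameter, and blue-masked accent areas add a Fresnel-weighted emissive glow.
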MaterialFunction /Game/Characters/Player/Male/Medium/Bodies/M_MED_Disguise_Black/Materials/MF_MED_Disguise_Black_AddPatterns
graph {nodes [
  n1 = TextureObjectParameter() {
   group=Skin FX
   parameterName=SkinFX_Mask
   texture=/Game/Characters/Player/Male/Medium/Bodies/M_MED_Disguise_Black/Textures/T_M_MED_Disguise_Black_Body_FX_base
  }
  n2 = TextureSample(Tex=@n1) {samplerSource=SSM_Wrap_WorldGroupSettings}
  Floor(None=@n2.R)
  n3 = GetMaterialAttributes(None=FunctionInput() {inputName=MA, inputType=FunctionInput_MaterialAttributes}) {attributeGetTypes [#1, #2, #3]}
  n4 = VectorParameter() {group=FX_Pattern, parameterName=Primary Color, defaultValue=<rgba 0.5365 0.5365 0.5365 0>}
  n5 = VectorParameter() {group=FX_Pattern, parameterName=Secondary Color, defaultValue=<rgba 0.5365 0.5365 0.5365 0>}
  n6 = OneMinus(None=@n2.R)
  n7 = LinearInterpolate(
   A=@n5.RGB
   B=LinearInterpolate(
    A=@n5.RGB
    B=Multiply(A=@n5.RGB, B=@n6) {desc="Multiply by Mask, for Camo"}
    Alpha=ScalarParameter() {group=FX_Pattern, parameterName=UseCamoBlend}
   )
   Alpha=@n2.G
  )
  n8 = LinearInterpolate(A=@n2.R, B=@n6, Alpha=@n2.R) {desc=Invert Colors when Green channel is empty. Ensures the shirt uses the secondary color.}
  n9 = VectorParameter() {group=FX_Pattern, parameterName=Accent Color, defaultValue=<rgba 0.5365 0.5365 0.5365 0>}
  n10 = Multiply(A=@n3.BaseColor, B=LinearInterpolate(A=LinearInterpolate(A=@n4.RGB, B=@n7, Alpha=@n8), B=@n9.RGB, Alpha=@n2.B))
  n11 = Saturate(None=Add(A=@n2.R, B=Add(A=@n2.G, B=@n2.G)))
  n12 = LinearInterpolate(A=@n3.SubsurfaceColor, B=@n10, Alpha=@n11)
  n13 = ScalarParameter() {group=Cloth, parameterName=Cloth Base Color Intensity}
  n14 = Multiply(
   A=Multiply(A=Multiply(A=@n2.B, B=@n9.RGB), B=Fresnel() {baseReflectFraction=0.03, exponent=1})
   B=ScalarParameter() {defaultValue=0.15, group=FX_Pattern, parameterName=AccentColorEmissive}
  ) {constB=0.02}
  FunctionOutput(None=SetMaterialAttributes!(
    MaterialAttributes=@n3.MaterialAttributes
    BaseColor=LinearInterpolate(A=@n3.BaseColor, B=@n10, Alpha=@n11)
    SubsurfaceColor=Multiply(A=@n12, B=@n13)
    EmissiveColor=Add(A=@n3.EmissiveColor, B=@n14)
   ) {attributeSetTypes [#1, #2, #3]})
 ]}
