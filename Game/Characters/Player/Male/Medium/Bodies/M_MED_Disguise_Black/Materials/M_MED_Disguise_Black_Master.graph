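# Master material for the black Disguise medium male player body, using material attributes with clothing, skeletal mesh, morph target and mesh deformer support.
# A shared character PreFX function feeds a Disguise-specific AddPatterns function, then CharacterSkinFX, shared PostFX and a cinematic post-FX function.
# PreFX also supplies auxiliary inputs to the skin and post stages.
# Shading model comes from the material expression, with a hero skin subsurface profile assigned.
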Material /Game/Characters/Player/Male/Medium/Bodies/M_MED_Disguise_Black/Materials/M_MED_Disguise_Black_Master
graph {
 properties {
  bAutomaticallySetUsageInEditor=false
  bUseMaterialAttributes=true
  bUsedWithClothing=true
  bUsedWithMeshDeformer=true
  bUsedWithMorphTargets=true
  bUsedWithSkeletalMesh=true
  pixelDepthOffsetMode=PDOM_Legacy
  refractionMethod=RM_IndexOfRefraction
  shadingModel=MSM_FromMaterialExpression
  subsurfaceProfile=&SS_HeroSkin_02
  usedShadingModels=Default Lit | Subsurface | Clear Coat | Subsurface Profile | Cloth | Eye
 }
 nodes [
  n1 = MaterialFunctionCall() {materialFunction=&BaseFN_CharacterFunction_PreFX}
  n2 = MaterialFunctionCall(
   AO=@n1.AO
   Distance Blend=@n1.AO
   Fresnel=@n1.AO
   Gmap=@n1.AO
   Pre Skinned Local Position=@n1.AO
   Material=MaterialFunctionCall(
    PreskinnedPosition=@n1.Pre skinned Local Position
    Gmap=@n1.Pre skinned Local Position
    MaterialAttributes=MaterialFunctionCall!(MA=@n1.Result) {materialFunction=&/MF_MED_Disguise_Black_AddPatterns}.Result
    AO=@n1.Pre skinned Local Position
   ) {materialFunction=&CharacterSkinFX}.Result
  ) {materialFunction=&BaseFN_CharacterFunction_PostFX}
  MaterialFunctionCall(In=@n2.Result) {materialFunction=&CINE_MF_CharPostFX}
 ]
}
